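# Translucent, surface-lit skeletal mesh material for the collar of the Slurp Alien Fishhead character skin, kept in a Demo folder.
# It builds material attributes from the Fishhead head's diffuse, normal and mask textures, with constant full opacity.
# The mask's red channel alone drives Specular, Roughness and Metallic.
# The attributes then pass through shared CharacterDeRez_Transition, BaseFN_CharacterOpacity and MF_TechArtOverrides functions in sequence.
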Material /BRCosmetics/Characters/Player/Male/Medium/Bodies/M_Med_Commando_AlienFishHead/Skins/Alien_Fishhead_Slurp/Demo/Collar
graph {
 properties {
  bUseMaterialAttributes=true
  bUsedWithMeshDeformer=true
  bUsedWithSkeletalMesh=true
  blendMode=BLEND_Translucent
  pixelDepthOffsetMode=PDOM_Legacy
  translucencyLightingMode=TLM_Surface
  usedShadingModels=Default Lit
 }
 nodes [
  n1 = TextureSample() {
   samplerType=SAMPLERTYPE_Masks
   texture=/Game/Characters/Player/Male/Medium/Heads/M_MED_AUX_Alien_Head_01/Textures/T_M_MED_AUX_Alien_Head_01_Fishhead_S
  }
  n2 = TextureSample() {
   samplerType=SAMPLERTYPE_Normal
   texture=/Game/Characters/Player/Male/Medium/Heads/M_MED_AUX_Alien_Head_01/Textures/T_M_MED_AUX_Alien_Head_01_Fishhead_N
  }
  n3 = SetMaterialAttributes!(
   MaterialAttributes
   Base Color=TextureSample() {texture=/Game/Characters/Player/Male/Medium/Heads/M_MED_AUX_Alien_Head_01/Textures/T_M_MED_AUX_Alien_Head_01_Fishhead_D}.RGB
   Specular=@n1.R
   Roughness=@n1.R
   Emissive Color
   Opacity=Constant() {r=1}
   Normal=@n2.RGB
   World Position Offset
   Metallic=@n1.R
  ) {attributeSetTypes [
    #1
    #2
    #3
    #4
    #5
    #6
    #7
    #8
   ]}
  MaterialFunctionCall(In=MaterialFunctionCall!(Material=MaterialFunctionCall!(Fresnel, Material=@n3) {materialFunction=/Game/Packages/Fortress_SharedMaterials/Base_Material_Functions/CharacterDeRez_Transition}.Result) {materialFunction=/Game/Packages/Fortress_SharedMaterials/Base_Material_Functions/BaseCharacterShaderFunctions/BaseFN_CharacterOpacity}.Result) {materialFunction=&MF_TechArtOverrides}
 ]
}
